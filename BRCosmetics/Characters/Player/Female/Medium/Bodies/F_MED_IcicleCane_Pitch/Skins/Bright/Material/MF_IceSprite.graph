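# Material function for the IcicleCane Pitch Bright skin that adds sparkling ice sprites to incoming material attributes, toggled by the UseIceSprite static switch.
# A tiled sprite texture supplies normals in RG and alpha in B.
# World-space sprite normals drive fresnel, view-based specular and focus highlights, tinted and limited by a makeup mask, then added to Emissive.
# A Custom node reads the directional light direction for an N·L lit term.
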MaterialFunction /BRCosmetics/Characters/Player/Female/Medium/Bodies/F_MED_IcicleCane_Pitch/Skins/Bright/Material/MF_IceSprite
graph {nodes [
  n1 = FunctionInput() {inputName=Material, inputType=FunctionInput_MaterialAttributes}
  n2 = GetMaterialAttributes(None=@n1) {attributeGetTypes [
    #1
    #2
    #3
    #4
    #5
   ]}
  n3 = SetMaterialAttributes!(
   MaterialAttributes=@n2.MaterialAttributes
   EmissiveColor=Add(A=@n2.EmissiveColor, B=NamedRerouteUsage())
   Roughness=@n2.MaterialAttributes
   Metallic=@n2.MaterialAttributes
   Normal=NamedRerouteUsage()
   BaseColor=NamedRerouteUsage()
  ) {attributeSetTypes [
    #2
    #4
    #5
    #1
    #3
   ]}
  FunctionOutput(None=StaticSwitchParameter(True=@n3, False=@n1) {group=IceSprite, parameterName=UseIceSprite})
  n4 = ScalarParameter() {defaultValue=15, group=IceSprite, parameterName=SpriteTilingX}
  n5 = ScalarParameter() {defaultValue=7.5, group=IceSprite, parameterName=SpriteTilingY}
  n6 = TextureSampleParameter2D(UVs=Multiply(A=TextureCoordinate(), B=AppendVector(A=@n4, B=@n5))) {
   group=IceSprite
   parameterName=IceSprite
   samplerType=SAMPLERTYPE_LinearColor
   texture=/BRCosmetics/Characters/Player/Female/Medium/Bodies/F_MED_IcicleCane_Pitch/Skins/Bright/Texture/T_IcySprite
  }
  n7 = Subtract(A=ComponentMask(None=@n6.RGB) {g=true, r=true}) {constB=0.5}
  n8 = Normalize(VectorInput=DeriveNormalZ(InXY=Multiply(A=@n7) {constB=2}))
  MaterialFunctionCall(Normal=@n8, Flatness=Constant()) {materialFunction=&FlattenNormal}
  Multiply()
  n9 = Transform(None=NamedRerouteUsage())
  n10 = Saturate(None=DotProduct(A=NamedRerouteUsage(), B=NamedRerouteDeclaration(None=@n9) {name=Normal_WS, nodeColor=<rgba 0.5234 1 0 1>}))
  n11 = ScalarParameter() {defaultValue=1, group=IceSprite, parameterName=LitPower}
  NamedRerouteDeclaration(None=SmoothStep(Value=Power(Base=@n10, Exp=@n11)) {constMax=0.5}) {name=Lit, nodeColor=<rgba 1 0.7031 0 1>}
  NamedRerouteDeclaration(None=@n2.Normal) {name=Normal_IN, nodeColor=<rgba 1 0.1172 0 1>}
  NamedRerouteDeclaration(None=@n6.B) {name=SpriteAlpha, nodeColor=<rgba 0.7812 0 1 1>}
  NamedRerouteDeclaration(None=Custom!(None) {code="return View.DirectionalLightDirection;", description=DirectionalLightDirection, showCode=true}) {name=LightVector_IN, nodeColor=<rgba 0 0.4141 1 1>}
  n12 = NamedRerouteDeclaration(None=Transform(None=NamedRerouteDeclaration(None=@n8) {name=Sprite_Normal_Tangent, nodeColor=<rgba 0.8516 1 0 1>})) {name=Sprite_Normal_WS, nodeColor=<rgba 0.5234 1 0 1>}
  n13 = Fresnel(Normal=@n12) {baseReflectFraction=0, exponent=1}
  n14 = LinearInterpolate(Alpha=Power(Base=NamedRerouteDeclaration(None=@n13) {name=Fresnel_Sprite, nodeColor=<rgba 0 0.9531 1 1>}) {constExponent=5}) {constA=0.2}
  n15 = ScalarParameter() {defaultValue=0.5, group=IceSprite, parameterName=SpritFresnelIntensity}
  n16 = Saturate(None=DotProduct(A=ReflectionVectorWS(CustomWorldNormal=@n12), B=CameraVectorWS()))
  n17 = ScalarParameter() {defaultValue=20, group=IceSprite, parameterName=SpecPOW_IceSprite}
  n18 = Power(Base=@n16, Exp=@n17)
  n19 = NamedRerouteUsage()
  n20 = Multiply(A=Multiply(A=Multiply(A=@n14, B=@n15), B=LinearInterpolate(Alpha=@n18) {constA=0.3}), B=@n19)
  n21 = NamedRerouteUsage()
  n22 = NamedRerouteUsage()
  n23 = Power(Base=DotProduct(A=ReflectionVectorWS(), B=CameraVectorWS()), Exp=@n17)
  n24 = ScalarParameter() {defaultValue=1, group=IceSprite, parameterName=SpecIntensity_Sprite}
  n25 = LinearInterpolate(
   A=Multiply(A=Multiply(A=@n23, B=@n24), B=ScalarParameter() {defaultValue=0.02, group=IceSprite, parameterName=SkinHighLight})
   B=Multiply(A=@n18, B=@n24)
   Alpha=@n19
  )
  n26 = ScalarParameter() {defaultValue=5, group=IceSprite, parameterName=SpecPOW_IceSprite_Focus}
  n27 = Multiply(A=Multiply(A=@n19, B=Power(Base=@n16, Exp=@n26)), B=ScalarParameter() {defaultValue=0.2, group=IceSprite, parameterName=FocuseIntensity})
  n28 = Add(A=Multiply(A=Multiply(A=@n20, B=@n21), B=@n22), B=Multiply(A=Add(A=@n25, B=@n27), B=@n22))
  n29 = VectorParameter() {group=IceSprite, parameterName=SpriteEMColor, defaultValue=<rgba 1 1 1 20>}
  n30 = Multiply(A=@n29.RGB, B=@n29.RGB)
  n31 = TextureSampleParameter2D() {group=IceSprite, parameterName=MakeUPMask, samplerType=SAMPLERTYPE_Masks, texture=&T_BlackMask}
  NamedRerouteDeclaration(None=Multiply(A=Multiply(A=@n28, B=@n30), B=@n31.R)) {name=SpriteEM, nodeColor=<rgba 1 0.5391 0 1>}
  NamedRerouteDeclaration(None=@n2.BaseColor) {name=BaseColorIN, nodeColor=<rgba 0.2188 0 1 1>}
  NamedRerouteUsage()
 ]}
