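# Black-ice material layer for the Fade skin of the IcicleCane Pitch character, blended over input material attributes by Alpha when UseBlackIce is on.
# Bump-offset ice noise and cracks give fake depth, with optional triplanar projection on pre-skinned coordinates.
# Fresnel-driven color curves, edge tint and cracks build time-of-day-scaled emissive over black base color; a surface normal drives roughness and normals.
# Facet-reflection and erosion branches do not reach the output.
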MaterialFunction /BRCosmetics/Characters/Player/Female/Medium/Bodies/F_MED_IcicleCane_Pitch/Skins/Fade/Materials/MF_Ice_Icicle
graph {nodes [
  n1 = ScalarParameter() {group=BlackIce, parameterName=IB_HeightOffset}
  n2 = ScalarParameter() {group=BlackIce, parameterName=IB_HeightRatio}
  n3 = TextureSampleParameter2D(UVs=BumpOffset(Coordinate=TextureCoordinate(), Height=@n1, HeightRatioInput=@n2)) {
   group=IceBreaker
   parameterName=IceNoise
   samplerSource=SSM_Wrap_WorldGroupSettings
   samplerType=SAMPLERTYPE_Normal
   texture=/Game/Weapons/WeaponSkins/Wraps/Textures/IceBreaker/T_IceNormal
  }
  n4 = TextureObjectParameter() {
   group=IceBreaker
   parameterName=IceNoise
   samplerType=SAMPLERTYPE_Normal
   texture=&T_IceNormal
  }
  n5 = MaterialFunctionCall!(
   TextureObject=@n4
   TextureSize=NamedRerouteUsage()
   WorldPosition=NamedRerouteUsage()
   Export Float 4
   World Space Normal=NamedRerouteUsage()
   ProjectionTransitionContrast
   Bump Height=@n1
   Bump Height Ratio=@n2
  ) {materialFunction=/Game/Characters/Player/Female/Medium/Bodies/F_MED_JurassicArchaeology/Skins/Winter/Materials/WorldAlignedTexture_WithBumpOffset}
  n6 = MaterialFunctionCall(Color=StaticSwitchParameter(True=@n5.XYZ Texture, False=@n3.RGB) {group=IceBreaker, parameterName=Use Triplanar}) {materialFunction=&SplitComponents}
  n7 = MaterialFunctionCall(Normal=@n6.RGB, Flatness=ScalarParameter() {defaultValue=0.4, group=IceBreaker, parameterName=IB_IceNormalFlatness}) {materialFunction=&FlattenNormal}
  n8 = Constant3Vector() {constant=<rgba 0 0 1 0>}
  n9 = Fresnel(
   ExponentIn=ScalarParameter() {defaultValue=2, group=BlackIce, parameterName=IB_FresnelExponent}
   Normal=Transform(None=MaterialFunctionCall(BaseNormal=@n8, AdditionalNormal=@n7.Result) {materialFunction=&BlendAngleCorrectedNormals}.Result)
  )
  n10 = ScalarParameter() {defaultValue=1.375, group=BlackIce, parameterName=IB_FresnelPower}
  n11 = Clamp(None=ComponentMask(None=@n6.RGB) {r=true})
  n12 = CameraVectorWS()
  n13 = VertexNormalWS()
  n14 = Clamp(None=Power(Base=DotProduct(A=@n12, B=@n13)) {constExponent=12})
  n15 = Multiply(
   A=ScalarParameter() {defaultValue=0.7864, group=BlackIce, parameterName=CurveAdjust}
   B=Max(A=Power(Base=@n9, Exp=@n10), B=Multiply(A=@n11, B=@n14))
  )
  Add(B=@n15)
  n16 = Multiply(A=CameraVectorWS()) {constB=-10}
  Add(B=MaterialFunctionCall(WorldPosition=Add(B=@n16)) {materialFunction=&WorldAlignedTexture}.XYZ Texture)
  n17 = TextureObjectParameter() {
   group=IceBreaker
   parameterName=ErosionNoise
   samplerType=SAMPLERTYPE_Masks
   texture=/Game/Weapons/WeaponSkins/Wraps/Textures/IceBreaker/T_IceBreakerCracks
  }
  n18 = NamedRerouteDeclaration(None=ScalarParameter() {defaultValue=-0.1, group=BlackIce, parameterName=IB_HeightOffset}) {name=IB_HeightOffset, nodeColor=<rgba 0.2656 1 0 1>}
  n19 = NamedRerouteDeclaration(None=ScalarParameter() {defaultValue=0.1, group=BlackIce, parameterName=IB_HeightRatio}) {name=IB_HeightRatio, nodeColor=<rgba 0.2656 1 0 1>}
  n20 = MaterialFunctionCall(
   TextureObject=@n17
   TextureSize=NamedRerouteUsage()
   WorldPosition=NamedRerouteUsage()
   World Space Normal=NamedRerouteUsage()
   Bump Height=@n18
   Bump Height Ratio=@n19
  ) {materialFunction=&WorldAlignedTexture_WithBumpOffset}
  n21 = TextureCoordinate()
  n22 = TextureSampleParameter2D(UVs=NamedRerouteDeclaration(None=BumpOffset(Coordinate=@n21, Height=@n18, HeightRatioInput=@n19)) {name=Shatter Bump Offset, nodeColor=<rgba 0 1 0.3438 1>}) {
   group=IceBreaker
   parameterName=ErosionNoise
   samplerSource=SSM_Wrap_WorldGroupSettings
   samplerType=SAMPLERTYPE_Masks
   texture=&T_IceBreakerCracks
  }
  n23 = StaticSwitchParameter(True=@n20.XYZ Texture, False=@n22.RGB) {group=IceBreaker, parameterName=Use Triplanar}
  n24 = Multiply(
   A=ChannelMaskParameter(None=@n23) {group=IceBreaker, maskChannel=Green, parameterName=TearNoisechannel}.R
   B=ScalarParameter() {defaultValue=1, group=IceBreaker, parameterName=IB_NoiseSpread}
  )
  Clamp(None=@n24)
  n25 = FunctionInput() {inputType=FunctionInput_MaterialAttributes}
  n26 = TextureObjectParameter() {group=IceBreaker, parameterName=IB_SurfaceNormal, samplerType=SAMPLERTYPE_Normal, texture=/Game/Weapons/WeaponSkins/Textures/StoneNormal}
  n27 = MaterialFunctionCall(
   TextureObject=@n26
   TextureSize=NamedRerouteUsage()
   WorldPosition=NamedRerouteUsage()
   World Space Normal=NamedRerouteUsage()
   Bump Height=NamedRerouteUsage()
   Bump Height Ratio=NamedRerouteUsage()
  ) {materialFunction=&WorldAlignedTexture_WithBumpOffset}
  n28 = TextureSampleParameter2D(UVs=NamedRerouteUsage()) {
   group=IceBreaker
   parameterName=IB_SurfaceNormal
   samplerSource=SSM_Wrap_WorldGroupSettings
   samplerType=SAMPLERTYPE_Normal
   texture=&StoneNormal
  }
  n29 = MaterialFunctionCall(Color=StaticSwitchParameter(True=@n27.XYZ Texture, False=@n28.RGB) {group=IceBreaker, parameterName=Use Triplanar}) {materialFunction=&SplitComponents}
  n30 = Power(Base=OneMinus(None=@n6.R))
  n31 = TextureSample(UVs=BumpOffset(
    Coordinate=TextureCoordinate() {uTiling=0.66, vTiling=0.66}
    Height=ScalarParameter() {defaultValue=-0.1, group=BlackIce, parameterName=IB_HeightOffset2}
    HeightRatioInput=ScalarParameter() {defaultValue=0.1, group=BlackIce, parameterName=IB_HeightRatio2}
   )) {
   samplerSource=SSM_Wrap_WorldGroupSettings
   samplerType=SAMPLERTYPE_Normal
   texture=/Game/Athena/Apollo/Environments/Landscape/Textures/Roads/Road_Cracks_N
  }
  n32 = Fresnel() {baseReflectFraction=0, exponent=0.15}
  n33 = Multiply(A=@n31.G, B=OneMinus(None=Saturate(None=@n32)))
  n34 = Multiply(A=Saturate(None=Multiply(A=@n30, B=@n33)), B=ScalarParameter() {defaultValue=0.5, group=BlackIce, parameterName=CrackIntensity})
  n35 = Constant()
  n36 = ScalarParameter() {defaultValue=0.668, group=BlackIce, parameterName=Base Emissive}
  n37 = NamedRerouteUsage()
  n38 = ScalarParameter() {defaultValue=1, group=IceBreaker, parameterName=Diffuse Bightness}
  n39 = CameraVectorWS()
  n40 = CurveAtlasRowParameter(CurveTime=Clamp(None=DotProduct(A=@n39, B=@n8))) {
   atlas=&AthenaCurveAtlas01
   curve=/Game/Weapons/WeaponSkins/Wraps/Textures/IceBreaker/IceReflectanceCurve
   defaultValue=198
   group=IceBreaker
   parameterName=IB_IceReflectanceCurve
  }
  n41 = ScalarParameter() {defaultValue=0.15, group=BlackIce, parameterName=GlowEXP}
  n42 = CurveAtlasRowParameter(CurveTime=Clamp(None=@n15)) {
   atlas=&AthenaCurveAtlas01
   curve=/Game/Weapons/WeaponSkins/Wraps/Textures/IceBreaker/IceBreakerColorCurve
   defaultValue=199
   group=BlackIce
   parameterName=IceBreakerBaseColorCurve
  }
  n43 = Multiply(A=OneMinus(None=Fresnel(ExponentIn=@n41) {baseReflectFraction=0}), B=@n42)
  n44 = Multiply(
   A=VectorParameter() {group=BlackIce, parameterName=IB_EdgeColor, defaultValue=<rgba 1 1 1 1>}.RGB
   B=ScalarParameter() {defaultValue=1, group=BlackIce, parameterName=IB_EdgeBrightness}
  )
  n45 = ScalarParameter() {defaultValue=1, group=BlackIce, parameterName=IB_BaseFresnelExponent}
  n46 = Transform(None=@n8)
  n47 = Clamp(None=Power(Base=Fresnel(ExponentIn=@n45, Normal=@n46), Exp=ScalarParameter() {defaultValue=2, group=BlackIce, parameterName=IB_BaseFresnelPower}))
  n48 = Multiply(
   A=LinearInterpolate(A=Constant() {r=1}, B=Multiply(A=@n37, B=@n38), Alpha=ScalarParameter() {group=IceBreaker, parameterName=Diffuse Influence})
   B=Multiply(A=@n40, B=LinearInterpolate(A=@n43, B=@n44, Alpha=@n47))
  )
  n49 = Multiply(
   A=Add(A=Multiply(A=@n35, B=@n9), B=Multiply(A=@n36, B=@n48))
   B=ScalarParameter() {defaultValue=2.5, group=BlackIce, parameterName=EmissiveStrength}
  )
  n50 = CurveAtlasRowParameter(CurveTime=CollectionParameter() {collection=&FortniteMaterialParameters, parameterName=TimeOfDay_0-1}) {
   atlas=/Game/Characters/Shared/CurveAtlas/AthenaCharacterCurveAtlas
   curve=/Game/Packages/Fortress_SharedMaterials/Anime_Master_Material/ToonTODBrightness_Curve
   defaultValue=17
   group=TOD_Adjustments
   parameterName=TODBrightnessAdjustmentCurve
  }
  n51 = SetMaterialAttributes!(
   MaterialAttributes=@n25
   BaseColor=Constant()
   Metallic=ScalarParameter() {defaultValue=0.24, group=BlackIce, parameterName=BlackIce_Metallic}
   Specular=ScalarParameter() {defaultValue=0.8871, group=BlackIce, parameterName=spec}
   Roughness=LinearInterpolate(
    A=ScalarParameter() {defaultValue=0.1, group=BlackIce, parameterName=BlackIce_Roughness, sliderMax=1}
    B=ScalarParameter() {defaultValue=0.3, group=BlackIce, parameterName=BlackIce_RoughnessMax, sliderMax=1}
    Alpha=@n29.G
   )
   Normal=MaterialFunctionCall(Normal=@n29.RGB, Flatness=ScalarParameter() {defaultValue=-0.5, group=BlackIce, parameterName=ice normal flatness}) {materialFunction=&FlattenNormal}.Result
   EmissiveColor=Multiply(A=Add(A=@n34, B=@n49), B=StaticComponentMaskParameter(None=@n50) {defaultR=true, parameterName=TODCurveChannelMask})
  ) {attributeSetTypes [
    #1
    #2
    #3
    #4
    #5
    #6
   ]}
  n52 = FunctionInput() {inputName=Alpha, inputType=FunctionInput_Scalar}
  FunctionOutput(None=StaticSwitchParameter(True=BlendMaterialAttributes(A=@n25, B=@n51, Alpha=@n52), False=@n25) {group=Skin FX, parameterName=UseBlackIce})
  MaterialFunctionCall(AdditionalNormal=MaterialFunctionCall(Normal=@n31.RGB) {materialFunction=&FlattenNormal}.Result) {materialFunction=&BlendAngleCorrectedNormals}
  n53 = TextureObjectParameter() {
   group=IceBreaker
   parameterName=IB_Facet_Normal
   samplerType=SAMPLERTYPE_Normal
   texture=/Game/Weapons/WeaponSkins/Wraps/Textures/IceBreaker/T_IceFacet_Normal
  }
  n54 = MaterialFunctionCall(
   TextureObject=@n53
   TextureSize=NamedRerouteUsage()
   WorldPosition=NamedRerouteUsage()
   World Space Normal=NamedRerouteUsage()
   Bump Height=NamedRerouteUsage()
   Bump Height Ratio=NamedRerouteUsage()
  ) {materialFunction=&WorldAlignedTexture_WithBumpOffset}
  n55 = TextureSampleParameter2D(UVs=NamedRerouteUsage()) {
   group=IceBreaker
   parameterName=IB_Facet_Normal
   samplerSource=SSM_Wrap_WorldGroupSettings
   samplerType=SAMPLERTYPE_Normal
   texture=&T_IceFacet_Normal
  }
  n56 = Transform(None=StaticSwitchParameter(True=@n54.XYZ Texture, False=@n55.RGB) {group=IceBreaker, parameterName=Use Triplanar}) {transformSourceType=TRANSFORMSOURCE_World, transformType=TRANSFORM_Tangent}
  n57 = TextureSampleParameter2D(UVs=ComponentMask(None=ReflectionVectorWS(CustomWorldNormal=@n56)) {g=true, r=true}) {
   parameterName=" IB_FakeReflectionNoise"
   samplerSource=SSM_Wrap_WorldGroupSettings
   texture=&T_Inky_Smoke_Tile_Chromatic
  }
  Multiply(A=@n57.RGB, B=VectorParameter() {group=BlackIce, parameterName=Reflection_Tint, defaultValue=<rgba 0.1 0.1 0.1 1>}.RGB)
  NamedRerouteDeclaration(None=VectorParameter() {group=IceBreaker, parameterName=Size, defaultValue=<rgba 100 100 100 1>}.RGB) {name=Triplanar Size, nodeColor=<rgba 1 0.4107 0 1>}
  NamedRerouteDeclaration(None=VertexInterpolator(VS=PreSkinnedPosition()).PS) {name=Triplanar Pos, nodeColor=<rgba 1 0 0.1641 1>}
  NamedRerouteDeclaration(None=VertexInterpolator(VS=PreSkinnedNormal()).PS) {name=Triplanar N, nodeColor=<rgba 0.6188 0 1 1>}
  NamedRerouteDeclaration(None=GetMaterialAttributes(None=@n25) {attributeGetTypes [#1]}.BaseColor) {name=Diffuse, nodeColor=<rgba 0.1953 1 0 1>}
  Power(Base=ChannelMaskParameter(None=@n23) {group=IceBreaker, parameterName=1}.R)
  Subtract(B=@n24)
  TextureCoordinate()
  TextureObject() {samplerType=SAMPLERTYPE_Normal, texture=&Road_Cracks_N}
 ]}
