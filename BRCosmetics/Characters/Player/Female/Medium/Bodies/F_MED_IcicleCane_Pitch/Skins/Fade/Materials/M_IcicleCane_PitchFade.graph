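# Character material for the Fade skin of the IcicleCane Pitch body, layering ice effects onto the shared 2023 character parent shading.
# The PreFX parent output feeds an icicle ice function, with a channel-mask parameter picking the Tech Art Mask channel as its alpha.
# An IceSprite function borrowed from the Bright skin follows, then the PostFX parent function completes the material attributes.
# It supports skeletal meshes, morph targets and mesh deformers.
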Material /BRCosmetics/Characters/Player/Female/Medium/Bodies/F_MED_IcicleCane_Pitch/Skins/Fade/Materials/M_IcicleCane_PitchFade
graph {
 properties {
  bAutomaticallySetUsageInEditor=false
  bUseMaterialAttributes=true
  bUsedWithMeshDeformer=true
  bUsedWithMorphTargets=true
  bUsedWithSkeletalMesh=true
  pixelDepthOffsetMode=PDOM_Legacy
  shadingModel=MSM_FromMaterialExpression
  usedShadingModels=Default Lit | Subsurface | Clear Coat | Subsurface Profile | Cloth
 }
 nodes [
  n1 = MaterialFunctionCall() {materialFunction=&MF_CharacterParent_2023_PreFX}
  n2 = ChannelMaskParameter(None=@n1.Tech Art Mask) {group=Skin FX, parameterName=BlackIceMaskChannel}
  MaterialFunctionCall(
   Material=MaterialFunctionCall!(Material=MaterialFunctionCall!(Alpha=@n2.R, In=@n1.Result) {materialFunction=&/MF_Ice_Icicle}.Result) {materialFunction=/BRCosmetics/Characters/Player/Female/Medium/Bodies/F_MED_IcicleCane_Pitch/Skins/Bright/Material/MF_IceSprite}.Result
   M=@n1.M
   Pre Skinned Local Position=@n1.M
   Distance Blend=@n1.M
   Fresnel=@n1.M
  ) {materialFunction=&MF_CharacterParent_2023_PostFX}
 ]
}
